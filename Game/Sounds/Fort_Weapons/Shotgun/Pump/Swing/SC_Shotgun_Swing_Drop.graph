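# Sound cue for the drop sound of the pump shotgun Swing variant.
# A random node picks between two equally weighted Paprika shotgun drop waves and feeds a modulator at the root.
# The cue slightly raises pitch and volume, uses weapon drop attenuation and the loot collect sound class, and joins the Weapon_Drop_StopOldest concurrency set.
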SoundCue /Game/Sounds/Fort_Weapons/Shotgun/Pump/Swing/SC_Shotgun_Swing_Drop
root=Modulator_0
nodes [
 {id=Modulator_0, children [Random_0]}
 {id=Random_0, children [WavePlayer_1,WavePlayer_2], properties {weights [1,1]}}
 {id=WavePlayer_1, properties {soundWaveAssetPtr=/Game/Sounds/Fort_Weapons/Shotgun/Pump/Swing/SoundWaves/Paprika_Shotgun_Drop_01}}
 {id=WavePlayer_2, properties {soundWaveAssetPtr=/Game/Sounds/Fort_Weapons/Shotgun/Pump/Swing/SoundWaves/Paprika_Shotgun_Drop_02}}
]
properties {
 attenuationSettings=&Fort_Weapon_Drop_Att
 maxDistance=3000
 pitchMultiplier=1.2
 soundClassObject=&Fort_Loot_Collect
 volumeMultiplier=1.1
 concurrencyOverrides {resolutionRule=PreventNew}
 concurrencySet [&Weapon_Drop_StopOldest]
}
