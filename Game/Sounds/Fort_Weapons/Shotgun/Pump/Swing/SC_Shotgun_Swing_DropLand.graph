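# Sound cue for a pump shotgun hitting the ground when dropped, built from generic shotgun drop impact waves.
# A random node picks between two equally weighted impact waves and feeds a modulator at the root.
# It plays in the loot collect sound class with its own short-range attenuation and a small concurrency cap that blocks new instances once reached.
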SoundCue /Game/Sounds/Fort_Weapons/Shotgun/Pump/Swing/SC_Shotgun_Swing_DropLand
root=Modulator_0
nodes [
 {id=Modulator_0, children [Random_0]}
 {id=Random_0, children [WavePlayer_0,WavePlayer_1], properties {weights [1,1]}}
 {
  id=WavePlayer_0
  properties {soundWaveAssetPtr=/Game/Sounds/Fort_Weapons/Shotgun/Pump/Swing/SoundWaves/Weapon_Generic_Drop_Impact_Shotgun_01}
 }
 {
  id=WavePlayer_1
  properties {soundWaveAssetPtr=/Game/Sounds/Fort_Weapons/Shotgun/Pump/Swing/SoundWaves/Weapon_Generic_Drop_Impact_Shotgun_02}
 }
]
properties {
 bOverrideAttenuation=true
 bOverrideConcurrency=true
 maxDistance=1000
 soundClassObject=&Fort_Loot_Collect
 volumeMultiplier=1
 attenuationOverrides {falloffDistance=1000, attenuationShapeExtents {x=0}}
 concurrencyOverrides {maxCount=3, resolutionRule=PreventNew}
}
