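# First-person fire sound cue for the pump "Swing" shotgun.
# A mixer layers a random pick of four main shots, a random pick of three mechanical sounds, and a single kick sample at reduced mix level.
# Modulators vary pitch and volume per play, most widely on the mechanical layer, whose volume can drop to zero.
# It uses first-person weapon attenuation, a ranged-fire sound class and shared gunfire concurrency.
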SoundCue /Game/Sounds/Fort_Weapons/Shotgun/Pump/Swing/SC_Shotgun_Swing_Fire_1P
root=Mixer_0
nodes [
 {id=Mixer_0, children [Modulator_1,Modulator_0,Modulator_2], properties {inputVolume [1,1,0.75]}}
 {id=Modulator_0, children [Random_0], properties {volumeMax=1, volumeMin=1}}
 {id=Modulator_1, children [Random_2], properties {pitchMax=2, volumeMax=1, volumeMin=0}}
 {id=Modulator_2, children [WavePlayer_4], properties {pitchMax=1.5, pitchMin=0.85, volumeMax=1, volumeMin=0.75}}
 {id=Random_0, children [WavePlayer_0,WavePlayer_1,WavePlayer_2,WavePlayer_3], properties {weights [1,1,1,1]}}
 {id=Random_2, children [WavePlayer_8, WavePlayer_9, WavePlayer_10], properties {weights [1,1,1]}}
 {id=WavePlayer_0, properties {soundWaveAssetPtr=/Game/Sounds/Fort_Weapons/Shotgun/Pump/Swing/SoundWaves/WPN_Shotgun_Swing_Fire_01}}
 {id=WavePlayer_1, properties {soundWaveAssetPtr=/Game/Sounds/Fort_Weapons/Shotgun/Pump/Swing/SoundWaves/WPN_Shotgun_Swing_Fire_02}}
 {id=WavePlayer_10, properties {soundWaveAssetPtr=/Game/Sounds/Fort_Weapons/Shotgun/Pump/Swing/SoundWaves/WPN_Shotgun_Swing_Fire_Mech_03}}
 {id=WavePlayer_2, properties {soundWaveAssetPtr=/Game/Sounds/Fort_Weapons/Shotgun/Pump/Swing/SoundWaves/WPN_Shotgun_Swing_Fire_03}}
 {id=WavePlayer_3, properties {soundWaveAssetPtr=/Game/Sounds/Fort_Weapons/Shotgun/Pump/Swing/SoundWaves/WPN_Shotgun_Swing_Fire_04}}
 {id=WavePlayer_4, properties {soundWaveAssetPtr=/Game/Sounds/Fort_Weapons/Shotgun/Pump/Swing/SoundWaves/WPN_Shotgun_Swing_Fire_Kick_01}}
 {id=WavePlayer_8, properties {soundWaveAssetPtr=/Game/Sounds/Fort_Weapons/Shotgun/Pump/Swing/SoundWaves/WPN_Shotgun_Swing_Fire_Mech_01}}
 {id=WavePlayer_9, properties {soundWaveAssetPtr=/Game/Sounds/Fort_Weapons/Shotgun/Pump/Swing/SoundWaves/WPN_Shotgun_Swing_Fire_Mech_02}}
]
properties {
 attenuationSettings=&Fort_Weapon_Projectile_FirstPerson_Athena_Att
 maxDistance=2097152
 soundClassObject=&Fort_WeaponRanged_Fire_1P_PSM
 volumeMultiplier=0.8
 attenuationOverrides {distanceAlgorithm=NaturalSound, falloffDistance=5000, attenuationShapeExtents {x=1000}}
 concurrencySet [&Weapon_Gunfire_FirstPerson_Conc]
}
