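# Sound cue for the pickup sound of the Swing pump shotgun, routed to the loot-collect sound class.
# A random node picks one of three equally weighted pickup waves and feeds a modulator for playback variation.
# Playback is quiet and pitched down, using shared weapon-pickup attenuation.
# A concurrency override allows one instance at a time and blocks new plays while it sounds.
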SoundCue /Game/Sounds/Fort_Weapons/Shotgun/Pump/Swing/SC_Shotgun_Swing_Pickup
root=Modulator_0
nodes [
 {id=Modulator_0, children [Random_0]}
 {id=Random_0, children [WavePlayer_0,WavePlayer_4,WavePlayer_5], properties {weights [1,1,1]}}
 {id=WavePlayer_0, properties {soundWaveAssetPtr=/Game/Sounds/Fort_Weapons/Shotgun/Pump/Swing/SoundWaves/Shotgun_Swing_Pickup_01}}
 {id=WavePlayer_4, properties {soundWaveAssetPtr=/Game/Sounds/Fort_Weapons/Shotgun/Pump/Swing/SoundWaves/Shotgun_Swing_Pickup_02}}
 {id=WavePlayer_5, properties {soundWaveAssetPtr=/Game/Sounds/Fort_Weapons/Shotgun/Pump/Swing/SoundWaves/Shotgun_Swing_Pickup_03}}
]
properties {
 attenuationSettings=&Fort_Weapon_Pickup_Att
 bOverrideConcurrency=true
 maxDistance=3500
 pitchMultiplier=0.8
 soundClassObject=&Fort_Loot_Collect
 volumeMultiplier=0.2
 concurrencyOverrides {maxCount=1, resolutionRule=PreventNew}
}
