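# SoundCue for the closing clasp foley of the pump Swing shotgun's reload.
# A single clasp wave plays through a Modulator left at its default settings.
# The cue is routed to the weapon foley sound class and uses the shared weapon reload attenuation at reduced volume.
# A concurrency override is enabled, with a rule that blocks new instances rather than cutting off playing ones.
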SoundCue /Game/Sounds/Fort_Weapons/Shotgun/Pump/Swing/SC_Shotgun_Swing_Reload_Close
root=Modulator_1
nodes [
 {id=Modulator_1, children [WavePlayer_2]}
 {
  id=WavePlayer_2
  properties {soundWaveAssetPtr=/Game/Sounds/Fort_Weapons/Shotgun/Pump/Swing/SoundWaves/WPN_Shotgun_Swing_Reload_Close_Clasp_01}
 }
]
properties {
 attenuationSettings=&Fort_Weapon_Reload_Att
 bOverrideConcurrency=true
 maxDistance=2250
 soundClassObject=&Fort_WeaponFoley
 volumeMultiplier=0.7
 attenuationOverrides {falloffDistance=4100, attenuationShapeExtents {x=900}}
 concurrencyOverrides {resolutionRule=PreventNew}
}
