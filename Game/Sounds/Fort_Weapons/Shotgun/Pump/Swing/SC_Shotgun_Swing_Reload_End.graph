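# Sound cue for the end of the pump shotgun swing reload, routed to the weapon foley sound class.
# A mixer layers three randomized, modulated groups: pump action, reload end and whoosh end.
# The pump layer is boosted in the mix and has its own upward pitch variation.
# It uses weapon reload attenuation and overrides concurrency to block new instances at the limit rather than cutting playing ones.
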SoundCue /Game/Sounds/Fort_Weapons/Shotgun/Pump/Swing/SC_Shotgun_Swing_Reload_End
root=Mixer_0
nodes [
 {id=Mixer_0, children [Modulator_0,Modulator_1,Modulator_2], properties {inputVolume [1.5,1,1]}}
 {id=Modulator_0, children [Random_2], properties {pitchMax=1.5, pitchMin=1}}
 {id=Modulator_1, children [Random_1]}
 {id=Modulator_2, children [Random_3]}
 {id=Random_1, children [WavePlayer_2,WavePlayer_3], properties {weights [1,1]}}
 {id=Random_2, children [WavePlayer_6,WavePlayer_7,WavePlayer_8], properties {weights [1,1,1]}}
 {id=Random_3, children [WavePlayer_5,WavePlayer_9], properties {weights [1,1]}}
 {id=WavePlayer_2, properties {soundWaveAssetPtr=/Game/Sounds/Fort_Weapons/Shotgun/Pump/Swing/SoundWaves/WPN_Shotgun_Swing_Reload_End_01}}
 {id=WavePlayer_3, properties {soundWaveAssetPtr=/Game/Sounds/Fort_Weapons/Shotgun/Pump/Swing/SoundWaves/WPN_Shotgun_Swing_Reload_End_02}}
 {
  id=WavePlayer_5
  properties {soundWaveAssetPtr=/Game/Sounds/Fort_Weapons/Shotgun/Pump/Swing/SoundWaves/WPN_Shotgun_Swing_Reload_Whoosh_End_01}
 }
 {id=WavePlayer_6, properties {soundWaveAssetPtr=/Game/Sounds/Fort_Weapons/Shotgun/Pump/Swing/SoundWaves/WPN_Shotgun_Swing_Reload_Pump_01}}
 {id=WavePlayer_7, properties {soundWaveAssetPtr=/Game/Sounds/Fort_Weapons/Shotgun/Pump/Swing/SoundWaves/WPN_Shotgun_Swing_Reload_Pump_02}}
 {id=WavePlayer_8, properties {soundWaveAssetPtr=/Game/Sounds/Fort_Weapons/Shotgun/Pump/Swing/SoundWaves/WPN_Shotgun_Swing_Reload_Pump_03}}
 {
  id=WavePlayer_9
  properties {soundWaveAssetPtr=/Game/Sounds/Fort_Weapons/Shotgun/Pump/Swing/SoundWaves/WPN_Shotgun_Swing_Reload_Whoosh_End_02}
 }
]
properties {
 attenuationSettings=&Fort_Weapon_Reload_Att
 bOverrideConcurrency=true
 maxDistance=2250
 soundClassObject=&Fort_WeaponFoley
 volumeMultiplier=0.501
 attenuationOverrides {falloffDistance=4100, attenuationShapeExtents {x=900}}
 concurrencyOverrides {resolutionRule=PreventNew}
}
